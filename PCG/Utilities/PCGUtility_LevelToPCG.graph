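# Editor utility that bakes each selected level into a saved "_PCG" settings asset using a PCGAsset element, creating or updating it.
# It converts static mesh components and instanced mesh instances into points with mesh, first material and actor tag attributes.
# Actor index, parent index, recursive depth and relative transform preserve attachment hierarchy.
# The asset outputs a Root pin (overall bounds, name, source path) and a Points pin.
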EditorUtilityBlueprint /PCG/Utilities/PCGUtility_LevelToPCG
graphs [
 ConstructPCGData {nodes [
   n1 = Inputs!() -> (execute, Outer:Object Reference)
   > n2 = Construct Object from Class(Class=/Script/PCG.PCGPointData, self=@n1) -> (ReturnValue:PCGPoint Data Object Reference)
   > n3 = Set PCGPointData_Root(PCGPointData_Root:PCGPoint Data Object Reference=@n2) -> (Output_Get:PCGPoint Data Object Reference)
   > n4 = MutableMetadata!(execute, self:PCGData Object Reference=@n3) -> (then, ReturnValue:PCGMetadata Object Reference)
   > Set PCGMetaData_Root(PCGMetaData_Root:PCGMetadata Object Reference=@n4) -> (Output_Get:PCGMetadata Object Reference)
   > n5 = Construct Object from Class(Class=&PCGPointData, self=@n1)
   > n6 = Set PCGPointData_Points(PCGPointData_Points:PCGPoint Data Object Reference=@n5) -> (Output_Get:PCGPoint Data Object Reference)
   > n7 = MutableMetadata(self=@n6)
   > Set PCGMetaData_Points(PCGMetaData_Points:PCGMetadata Object Reference=@n7) -> (Output_Get:PCGMetadata Object Reference)
   > Outputs!(then:Exec)
  ]}
 EventGraph {links [], nodes []}
 GetHierarchyDepth {nodes [
   n1 = GetHierarchyDepth!() -> (then, Actor:Actor Object Reference, InitialDepth:int) {is=title K2Node_FunctionEntry}
   n2 = GetAttachParentActor!(self:Actor Object Reference=@n1) -> (ReturnValue:Actor Object Reference)
   > n3 = Is Valid(InputObject=@n2)
   Return Node!(execute=@n3.Is Not Valid, Depth:int=@n1.InitialDepth)
   > n4 = GetHierarchyDepth!(execute, self, Actor:Actor Object Reference=@n2, InitialDepth:int=int + int(A=@n1.InitialDepth)) -> (then, Depth:int=0) {is=title K2Node_CallFunction}
   > Return Node(Depth=@n4)
  ]}
 PCG - Level to PCG Settings {nodes [
   PCG - Level to PCG Settings() {is=title K2Node_FunctionEntry}
   > n1 = GetSelectedAssets()
   > n2 = For Each Loop(Array:Array of Object References=@n1) -> (Array Element:Object Reference)
   > Set Level(Level:Object Reference=@n2) -> (Output_Get:Object Reference)
   n3 = Get Level() -> (Level:Object Reference)
   n4 = GetDisplayName(Object=@n3)
   > Set Name(Name:String=@n4) -> (Output_Get:String)
   n5 = Replace(SourceString=Get Object Path String(Object=@n3), From=Append!(A=@n4, B=., C:String=@n4) -> (ReturnValue))
   > Set FolderPath(FolderPath:String=@n5) -> (Output_Get:String)
   n6 = Append!(A=@n4, B=_PCG) -> (ReturnValue)
   > Set FileName(FileName:String=@n6) -> (Output_Get:String)
   > Set FilePath(FilePath:String=Append!(A=@n5, B=@n6, C:String) -> (ReturnValue)) -> (Output_Get:String)
   > n7 = DoesAssetExist(AssetPath=Get FilePath() -> (FilePath:String))
   > n8 = Set FileExist(FileExist:bool=@n7) -> (Output_Get:bool)
   > n9 = Branch(Condition=@n8)
   > n10 = LoadAsset(AssetPath=Get FilePath())
   > n11 = Set OutputAsset(OutputAsset:Object Reference=@n10) -> (Output_Get:Object Reference)
   > n12 = Construct PCGData!(execute:Exec, Outer:Object Reference=@n11) -> (then:Exec) {is=title K2Node_MacroInstance}
   Clear(execute=[@n78.then, @n12.then], TargetArray:Array of Names=Get TagsAttributes() -> (TagsAttributes:Array of Names))
   > Clear(TargetArray=Get Tags() -> (Tags:Array of Names))
   n13 = Make Literal Float(Value=10000000000000000)
   > Set!(execute, ActorBoundsMin_X:double=@n13, ActorBoundsMin_Y:double=@n13, ActorBoundsMin_Z:double=@n13) -> (then, Output_Get:Vector)
   n14 = Make Literal Float(Value=-10000000000000000)
   > Set!(execute, ActorBoundsMax_X:double=@n14, ActorBoundsMax_Y:double=@n14, ActorBoundsMax_Z:double=@n14) -> (then, Output_Get)
   > Clear!(
    execute
    TargetMap:Map of Actor Object References to Integers=Get ActorIndexMap() -> (ActorIndexMap:Map of Actor Object References to Integers)
   ) -> (then) {is=title K2Node_CallFunction}
   > Clear(TargetArray:Array of PCGPin Properties Structures=Get OutputPins() -> (OutputPins:Array of PCGPin Properties Structures)) {is=title K2Node_CallArrayFunction}
   > Clear(TargetArray:Array of PCGPoint Structures=Get PCGPoints() -> (PCGPoints:Array of PCGPoint Structures))
   > n15 = Sequence!(execute) -> (then_0, then_1, then_2:Exec, then_3:Exec)
   n16 = GetAllActorsOfClass!(execute=@n15.then_1, WorldContextObject:Object Reference=Get Level(), ActorClass:Actor Class Reference=&Actor) -> (then, OutActors:Array of Actor Object References)
   > n17 = For Each Loop(Array:Array of Actor Object References=@n16) -> (Array Element:Actor Object Reference)
   > Add!(
    execute
    TargetMap:Map of Actor Object References to Integers=Get ActorIndexMap()
    Key:"Actor Object Reference (by ref)"=@n17
    Value:"Integer (by ref)"=int + int(A=@n17.Array Index)
   ) -> (then) {is=title K2Node_CallFunction}
   n18 = Get PCGPoints()
   n19 = Add(
    execute=@n15.then_2
    TargetArray:Array of PCGPoint Structures=@n18
    NewItem:"PCGPoint Structure (by ref)"=Make PCGPoint!(
     Transform:Transform=MakeTransform(Rotation="0, 0, 0")
     Density:float=1
     BoundsMin:Vector="0,0,0"
     BoundsMax:Vector="0,0,0"
     Color:Vector 4 Structure=(X=1,Y=1,Z=1,W=1)
     Steepness:float=1
     Seed:int=0
    ) -> (PCGPoint:PCGPoint Structure) {is=title K2Node_MakeStruct}
   ) {is=title K2Node_CallArrayFunction}
   > SetInteger64Attribute!(
    execute
    Point:"PCGPoint Structure (by ref)"="Get (a ref)"(Array:Array of PCGPoint Structures=@n18, Dimension 1=@n19) -> (Output:"PCGPoint Structure (by ref)")
    Metadata:PCGMetadata Object Reference=Get PCGMetaData_Points() -> (PCGMetaData_Points:PCGMetadata Object Reference)
    AttributeName:Name=ActorIndex
    Value:Integer64=0
   ) -> (then)
   n20 = GetAllActorsOfClass(execute=@n15.then_3, WorldContextObject=Get Level(), ActorClass=&Actor)
   > n21 = For Each Loop(Array=@n20)
   n22 = MakeTransform(Rotation="0, 0, 0")
   n23 = Get ActorBoundsMax() -> (ActorBoundsMax:Vector)
   n24 = Get ActorBoundsMin() -> (ActorBoundsMin:Vector)
   SetPoints!(
    execute=@n21.Completed
    self:PCGPoint Data Object Reference=Get PCGPointData_Root() -> (PCGPointData_Root:PCGPoint Data Object Reference)
    InPoints:Array of PCGPoint Structures=Make Array("[0]":PCGPoint Structure=Make PCGPoint(Transform=@n22, Density=1, BoundsMin=@n24, BoundsMax=@n23, Color=(X=1,Y=1,Z=1,W=1), Steepness=1, Seed=0)) -> (Array:Array of PCGPoint Structures)
   ) -> (then)
   > SetPoints(self=Get PCGPointData_Points() -> (PCGPointData_Points:PCGPoint Data Object Reference), InPoints=Get PCGPoints())
   n25 = MakeLiteralName(Value=Root)
   n26 = MakeLiteralName(Value=Points)
   > n27 = For Each Loop(Array:Array of Names=Make Array!("[0]":Name=@n25, "[1]":Name=@n26) -> (Array:Array of Names)) -> (Array Element:Name)
   n28 = MakeTaggedData!(Data:PCGData Object Reference=Get PCGPointData_Root(), Tags:Set of Strings, Pin:Name=@n25) -> (ReturnValue:PCGTagged Data Structure)
   n29 = MakeTaggedData(Data=Get PCGPointData_Points(), Pin=@n26)
   Set PCGDataCollectionOutput(
    execute=@n27.Completed
    PCGDataCollectionOutput:PCGData Collection Structure=Make PCGData Collection!(
     TaggedData:Array of PCGTagged Data Structures=Make Array("[0]":PCGTagged Data Structure=@n28, "[1]":PCGTagged Data Structure=@n29) -> (Array:Array of PCGTagged Data Structures)
     bCancelExecutionOnEmpty:bool=False
    ) -> (PCGDataCollection:PCGData Collection Structure) {is=title K2Node_MakeStruct}
   ) -> (Output_Get:PCGData Collection Structure)
   > n30 = Branch(Condition=Get FileExist() -> (FileExist:bool))
   > n31 = Cast To PCGBlueprintSettings!(execute, Object=Get OutputAsset() -> (OutputAsset:Object Reference)) -> (then, CastFailed, AsPCGBlueprint Settings:PCGBlueprint Settings Object Reference) {is=title K2Node_DynamicCast}
   > n32 = Cast To PCGAsset!(
    execute
    Object=Get BlueprintElementInstance!(self:PCGBlueprint Settings Object Reference=@n31) -> (BlueprintElementInstance:PCGBlueprint Base Element Object Reference)
   ) -> (then, CastFailed, AsPCGAsset:PCGAsset Object Reference) {is=title K2Node_DynamicCast}
   > n33 = Set OutputPCGAsset(OutputPCGAsset:PCGAsset Object Reference=@n32) -> (Output_Get:PCGAsset Object Reference)
   n34 = Get OutputPCGAsset() -> (OutputPCGAsset:PCGAsset Object Reference)
   Set CustomOutputPins!(
    execute=[@n38.then, @n33.then]
    CustomOutputPins:Array of PCGPin Properties Structures=Get OutputPins()
    self:PCGBlueprint Base Element Object Reference=@n34
   ) -> (then, Output_Get:Array of PCGPin Properties Structures)
   > Set bHasDefaultInPin!(execute, bHasDefaultInPin:bool=false, self:PCGBlueprint Base Element Object Reference=@n34) -> (then, Output_Get:bool)
   > Set bHasDefaultOutPin!(execute, bHasDefaultOutPin:bool=false, self:PCGBlueprint Base Element Object Reference=@n34) -> (then, Output_Get:bool)
   n35 = Get OutputPCGAsset()
   > Set PCGDataCollection!(
    execute
    PCGDataCollection:PCGData Collection Structure=Get PCGDataCollectionOutput() -> (PCGDataCollectionOutput:PCGData Collection Structure)
    self:PCGAsset Object Reference=@n35
   ) -> (then, Output_Get:PCGData Collection Structure)
   > Set NodeTitle!(execute, NodeTitle:Name=String To Name(InString=Get Name() -> (Name:String)), self:PCGAsset Object Reference=@n35) -> (then, Output_Get:Name=None)
   > SaveLoadedAsset(AssetToSave=Get OutputAsset(), bOnlyIfIsDirty=false)
   n36 = Cast To PCGBlueprintSettings(execute=@n30.else, Object=Get OutputAsset())
   > n37 = SetElementType!(
    execute
    self:PCGBlueprint Settings Object Reference=@n36
    InElementType:Deprecated Blueprint Element Class Reference=/PCG/BP_Elements/PCGAsset.PCGAsset_C
   ) -> (then, ElementInstance:PCGAsset Object Reference)
   > n38 = Set OutputPCGAsset(OutputPCGAsset=@n37)
   Add(
    execute=@n27.LoopBody
    TargetArray:Array of PCGPin Properties Structures=Get OutputPins()
    NewItem:"PCGPin Properties Structure (by ref)"=MakePinProperty!(
     Label:Name=@n27
     bAllowMultipleData:bool=True
     bAllowMultipleConnections:bool=True
     bIsAdvancedPin:bool=false
     AllowedType:EPCGExclusiveDataType Enum=Point
    ) -> (ReturnValue:PCGPin Properties Structure)
   )
   Set CurrentActor(execute=@n21.LoopBody, CurrentActor:Actor Object Reference=@n21) -> (Output_Get:Actor Object Reference)
   n39 = Get Components By Class!(
    self:Actor Object Reference=Get CurrentActor() -> (CurrentActor:Actor Object Reference)
    ComponentClass:Actor Component Class Reference=&StaticMeshComponent
   ) -> (ReturnValue:Array of Static Mesh Component Object References)
   > Branch(Condition="integer > integer"(A=Length(TargetArray:Array of Static Mesh Component Object References=@n39)))
   > Set ActorID(ActorID:int=@n21.Array Index) -> (Output_Get:int=0)
   > n40 = GetActorBoundsPCG!(execute, InActor:Actor Object Reference=Get CurrentActor(), bIgnorePCGCreatedComponents:bool=true) -> (then, ReturnValue_Min:Vector=~, ReturnValue_Max:Vector=~, ReturnValue_IsValid:bool=~)
   > Set ActorBoundsMin(ActorBoundsMin:Vector=Vector_ComponentMin!(A:Vector=@n40, B:Vector=Get ActorBoundsMin()) -> (ReturnValue:Vector)) -> (Output_Get:Vector)
   > Set ActorBoundsMax(ActorBoundsMax:Vector=Vector_ComponentMax!(A:Vector=@n40.ReturnValue_Max, B:Vector=Get ActorBoundsMax()) -> (ReturnValue:Vector)) -> (Output_Get:Vector)
   n41 = Get CurrentActor()
   > Set Tags(Tags:Array of Names=Get Tags!(self:Actor Object Reference=@n41) -> (Tags)) -> (Output_Get:Array of Names)
   > Set PointTransform(PointTransform:Transform=Get Actor Transform(self=@n41)) -> (Output_Get:Transform)
   n42 = Get Tags!() -> (Tags)
   > n43 = Branch(Condition="integer > integer"(A=Length(TargetArray:Array of Names=@n42)))
   > n44 = For Each Loop(Array=Get Tags())
   n45 = Get TagsAttributes()
   > n46 = Add Unique(TargetArray:Array of Names=@n45, NewItem:"Name (by ref)"=@n44)
   > Branch(Condition=Is Valid Index!(TargetArray:Array of Names=@n45, IndexToTest:int=@n46) -> (ReturnValue:bool))
   > CreateInteger64Attribute!(
    execute
    self:PCGMetadata Object Reference=Get PCGMetaData_Points()
    AttributeName:Name="Get (a ref)"(Array:Array of Names=@n45, Dimension 1=@n46) -> (Output:"Name (by ref)")
    DefaultValue:Integer64=0
    bAllowsInterpolation:bool=false
    bOverrideParent:bool=true
   ) -> (then, ReturnValue:PCGMetadata Object Reference)
   n47 = For Each Loop(
    Exec=[@n43.else, @n44.Completed]
    Array:Array of Static Mesh Component Object References=Get Components By Class(self=Get CurrentActor(), ComponentClass=&StaticMeshComponent)
   ) -> (Array Element:Static Mesh Component Object Reference)
   > n48 = Set StaticMesh(StaticMesh:Static Mesh Object Reference=Get StaticMesh(self=@n47)) -> (Output_Get:Static Mesh Object Reference)
   > Is Valid(InputObject=@n48)
   n49 = GetMaterials!(self:Mesh Component Object Reference=@n47) -> (ReturnValue:Array of Material Interface Object References)
   > n50 = Branch(Condition="integer > integer"(A=Length(TargetArray:Array of Material Interface Object References=@n49)))
   > n51 = Set Material(Material:Material Interface Object Reference="Get (a copy)"(Array:Array of Material Interface Object References=@n49) -> (Output:Material Interface Object Reference)) -> (Output_Get:Material Interface Object Reference)
   n52 = Branch(
    execute=[@n50.else, @n51.then]
    Condition="Equal (Class)"(A=Get Class(Object=@n47) -> (ReturnValue:Static Mesh Component Class Reference), B=&StaticMeshComponent)
   )
   n53 = Get PointTransform() -> (PointTransform:Transform)
   n54 = ComputeSeedFromPosition!(execute=[@n52.then, @n74.then], InPosition:"Vector (by ref)"=BreakTransform(InTransform=@n53)) -> (then, ReturnValue:int=0)
   n55 = GetBoundingBox!(self:Static Mesh Object Reference=Get StaticMesh!() -> (StaticMesh)) -> (ReturnValue_Min:Vector=~, ReturnValue_Max:Vector=~, ReturnValue_IsValid:bool=~)
   n56 = Get PCGPoints()
   > n57 = Add(
    TargetArray:Array of PCGPoint Structures=@n56
    NewItem:"PCGPoint Structure (by ref)"=Make PCGPoint(Transform=@n53, Density=1, BoundsMin=@n55, BoundsMax=@n55.ReturnValue_Max, Color=(X=1,Y=1,Z=1,W=1), Steepness=1, Seed=@n54)
   )
   n58 = "Get (a ref)"(Array:Array of PCGPoint Structures=@n56, Dimension 1=@n57) -> (Output:"PCGPoint Structure (by ref)")
   > SetSoftObjectPathAttribute!(
    execute
    Point:"PCGPoint Structure (by ref)"=@n58
    Metadata:PCGMetadata Object Reference=Get PCGMetaData_Points()
    AttributeName:Name=Mesh
    Value:"Soft Object Path Structure (by ref)"=GetSoftObjectPath(Object=Get StaticMesh())
   ) -> (then)
   n59 = Get Material() -> (Material:Material Interface Object Reference)
   > n60 = Is Valid(InputObject=@n59)
   n61 = Get CurrentActor()
   n62 = Get ActorIndexMap()
   SetInteger64Attribute(
    execute=[@n71.then, @n60.Is Not Valid]
    Point=@n58
    Metadata=Get PCGMetaData_Points()
    AttributeName=ActorIndex
    Value="To Integer64 (Integer)"!(InInt:int=Find(TargetMap:Map of Actor Object References to Integers=@n62, Key:"Actor Object Reference (by ref)"=@n61) -> (Value:int)) -> (ReturnValue:Integer64=0)
   )
   > n63 = GetHierarchyDepth(Actor=Get CurrentActor(), InitialDepth=1) -> (Depth=0)
   > n64 = Set ActorHierarchyDepth(ActorHierarchyDepth:int=@n63) -> (Output_Get:int=0)
   > SetInteger64Attribute(Point=@n58, Metadata=Get PCGMetaData_Points(), AttributeName=HierarchyDepth, Value="To Integer64 (Integer)"(InInt=@n64) -> (ReturnValue=0))
   n65 = GetAttachParentActor(self=Get CurrentActor())
   n66 = Find(TargetMap=Get ActorIndexMap(), Key=@n65)
   > SetInteger64Attribute(
    Point=@n58
    Metadata=Get PCGMetaData_Points()
    AttributeName=ParentIndex
    Value=Select(Option 0:Integer64=0, Option 1:Integer64="To Integer64 (Integer)"(InInt=@n66) -> (ReturnValue=0), Index=@n66.ReturnValue) -> (ReturnValue:Integer64)
   )
   n67 = Get Actor Transform(self=Get CurrentActor())
   n68 = Get Actor Transform(self=@n65)
   n69 = Get ActorHierarchyDepth() -> (ActorHierarchyDepth:int=0)
   > SetTransformAttribute!(
    execute
    Point:"PCGPoint Structure (by ref)"=@n58
    Metadata:PCGMetadata Object Reference=Get PCGMetaData_Points()
    AttributeName:Name=RelativeTransform
    Value:"Transform (by ref)"=SelectTransform!(
     A:"Transform (by ref)"=MakeRelativeTransform!(A:"Transform (by ref)"=@n67, RelativeTo:"Transform (by ref)"=@n68) -> (ReturnValue:Transform)
     B:"Transform (by ref)"=@n67
     bPickA:bool="integer > integer"(A=@n69, B=1)
    ) -> (ReturnValue:Transform)
   ) -> (then)
   > n70 = For Each Loop(Array:Array of Names=Get Tags()) -> (Array Element:Name)
   > SetInteger64Attribute(Point=@n58, Metadata=Get PCGMetaData_Points(), AttributeName=@n70, Value=1)
   n71 = SetSoftObjectPathAttribute(execute=@n60.Is Valid, Point=@n58, Metadata=Get PCGMetaData_Points(), AttributeName=Material, Value=GetSoftObjectPath(Object=@n59))
   n72 = Cast To InstancedStaticMeshComponent!(execute=@n52.else, Object=@n47) -> (then, CastFailed, AsInstanced Static Mesh Component:Instanced Static Mesh Component Object Reference) {is=title K2Node_DynamicCast}
   > n73 = For Loop(LastIndex=int - int(A=GetInstanceCount!(self:Instanced Static Mesh Component Object Reference=@n72) -> (ReturnValue:int=0)))
   > n74 = Set PointTransform(PointTransform=GetInstanceTransform!(self:Instanced Static Mesh Component Object Reference=@n72, InstanceIndex:int=@n73, bWorldSpace:bool=true) -> (OutInstanceTransform:Transform, ReturnValue:bool))
   CreateStringAttribute!(
    execute=@n15.then_0
    self:PCGMetadata Object Reference=Get PCGMetaData_Root() -> (PCGMetaData_Root:PCGMetadata Object Reference)
    AttributeName:Name=Name
    DefaultValue:String=Get Name()
    bAllowsInterpolation:bool=false
    bOverrideParent:bool=true
   ) -> (then, ReturnValue:PCGMetadata Object Reference)
   > CreateSoftObjectPathAttribute!(
    execute
    self:PCGMetadata Object Reference=Get PCGMetaData_Root()
    AttributeName:Name=Source
    DefaultValue_PathString:String=Get FilePath()
    bAllowsInterpolation:bool=false
    bOverrideParent:bool=true
   ) -> (then, ReturnValue:PCGMetadata Object Reference)
   > CreateSoftObjectPathAttribute(self=Get PCGMetaData_Points(), AttributeName=Material, bAllowsInterpolation=false, bOverrideParent=true)
   > CreateSoftObjectPathAttribute(self=Get PCGMetaData_Points(), AttributeName=Mesh, bAllowsInterpolation=false, bOverrideParent=true)
   > CreateInteger64Attribute(self=Get PCGMetaData_Points(), AttributeName=HierarchyDepth, DefaultValue=0, bAllowsInterpolation=false, bOverrideParent=true)
   > CreateInteger64Attribute(self=Get PCGMetaData_Points(), AttributeName=ActorIndex, DefaultValue=-1, bAllowsInterpolation=false, bOverrideParent=true)
   > CreateInteger64Attribute(self=Get PCGMetaData_Points(), AttributeName=ParentIndex, DefaultValue=-1, bAllowsInterpolation=false, bOverrideParent=true)
   > CreateTransformAttribute!(
    execute
    self:PCGMetadata Object Reference=Get PCGMetaData_Points()
    AttributeName:Name=RelativeTransform
    DefaultValue:Transform=MakeTransform(Rotation="0, 0, 0")
    bAllowsInterpolation:bool=false
    bOverrideParent:bool=true
   ) -> (then, ReturnValue:PCGMetadata Object Reference)
   n75 = GetAssetTools(execute=@n9.else)
   > n76 = CreateAsset(
    self=@n75
    AssetName=Get FileName() -> (FileName:String)
    PackagePath=Get FolderPath() -> (FolderPath:String)
    AssetClass=/Script/PCG.PCGBlueprintSettings
   )
   > n77 = Set OutputAsset(OutputAsset=@n76)
   > n78 = Construct PCGData(Outer=@n77)
   Get Object Path String(Object=@n59)
  ]}
]
variables [
 ActorBoundsMax {type=Vector}
 ActorBoundsMin {type=Vector}
 CurrentActor {type=Actor, default=null}
 Level {type=Object, default=null}
 OutputAsset {type=Object, default=null}
 PCGDataCollectionOutput {type=PCGDataCollection, default {bCancelExecutionOnEmpty=false, taggedData []}}
 PCGMetadata {type=PCGMetadata, default=null}
 PCGMetaData_Points {type=PCGMetadata, default=null}
 PCGMetaData_Root {type=PCGMetadata, default=null}
 PCGPointData {type=PCGPointData, default=null}
 PCGPointData_Points {type=PCGPointData, default=null}
 PCGPointData_Root {type=PCGPointData, default=null}
 PCGPoints {type="array<PCGPoint>", default []}
 PointTransform {type=Transform, default {location=<0 0 0>, rotation=<rot 0 0 0>, scale=<1 1 1>}}
 Tags {type="array<string>", default []}
 TagsAttributes {type="array<string>", default []}
]
